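# Material function that lets impostor baking render a chosen material attribute, normal or depth through emissive.
# MPC_ImposterCapture weights sum attribute channels with packed normals (optional tangent-space transform and two-sided sign) and depth from the capture basis, origin and radius.
# EnableCapture zeroes several lit inputs and adds that sum to emissive; CustomLighting adds to base color.
# A static switch bypasses everything; basis-vector debug text never reaches the output.
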MaterialFunction /ImpostorBaker/MaterialFunctions/ImposterCaptureSwitch
graph {
 properties {bExposeToLibrary=true, description=Used for rendering Imposors}
 nodes [
  n1 = FunctionInput() {inputName=Material Attributes, inputType=FunctionInput_MaterialAttributes}
  n2 = BreakMaterialAttributes!(Attr=@n1)
  n3 = Add(
   A=Multiply(A=@n2.BaseColor, B=CollectionParameter() {collection=/ImpostorBaker/Materials/MPC/MPC_ImposterCapture, parameterName=BaseColor})
   B=Multiply(A=@n2.Metallic, B=CollectionParameter() {collection=&MPC_ImposterCapture, parameterName=Metallic})
  )
  n4 = Add(
   A=Multiply(A=@n2.Specular, B=CollectionParameter() {collection=&MPC_ImposterCapture, parameterName=Specular})
   B=Multiply(A=@n2.Roughness, B=CollectionParameter() {collection=&MPC_ImposterCapture, parameterName=Roughness})
  )
  n5 = Add(
   A=Multiply(A=@n2.Anisotropy, B=CollectionParameter() {collection=&MPC_ImposterCapture, parameterName=Anisotropy})
   B=Multiply(A=@n2.EmissiveColor, B=CollectionParameter() {collection=&MPC_ImposterCapture, parameterName=Emissive})
  )
  n6 = Add(
   A=Multiply(A=@n2.Opacity, B=CollectionParameter() {collection=&MPC_ImposterCapture, parameterName=Opacity})
   B=Multiply(A=@n2.Tangent, B=CollectionParameter() {collection=&MPC_ImposterCapture, parameterName=Tangent})
  )
  n7 = Multiply(A=@n2.SubsurfaceColor, B=CollectionParameter() {collection=&MPC_ImposterCapture, parameterName=SubSurface})
  n8 = Multiply(A=@n2.ClearCoat, B=CollectionParameter() {collection=&MPC_ImposterCapture, parameterName=ClearCoat})
  n9 = Multiply(
   A=@n2.ClearCoatRoughness
   B=CollectionParameter() {collection=&MPC_ImposterCapture, parameterName=ClearCoatRoughness}
  )
  n10 = Multiply(
   A=@n2.AmbientOcclusion
   B=CollectionParameter() {collection=&MPC_ImposterCapture, parameterName=AmbientOcclusion}
  )
  n11 = Add(A=Add(A=Add(A=@n3, B=@n4), B=Add(A=@n5, B=@n6)), B=Add(A=Add(A=@n7, B=@n8), B=Add(A=@n9, B=@n10)))
  Add(
   A=@n11
   B=Multiply(A=CollectionParameter() {collection=&MPC_ImposterCapture, parameterName=BaseColor}, B=NamedRerouteUsage())
  )
  n12 = GetMaterialAttributes(None=@n1) {attributeGetTypes [
    #1
    #2
    #3
    #4
    #5
    #6
   ]}
  n13 = OneMinus(None=CollectionParameter() {collection=&MPC_ImposterCapture, parameterName=EnableCapture})
  n14 = SetMaterialAttributes!(
   MaterialAttributes=@n12.MaterialAttributes
   BaseColor=Multiply(A=@n12.BaseColor, B=@n13)
   Metallic=Multiply(A=@n12.Metallic, B=@n13)
   Specular=Multiply(A=@n12.Specular, B=@n13)
   EmissiveColor=Multiply(A=@n12.EmissiveColor, B=@n13)
   SubsurfaceColor=Multiply(A=@n12.SubsurfaceColor, B=@n13)
   CustomData0=Multiply(A=@n12.CustomData0, B=@n13)
  ) {attributeSetTypes [
    #1
    #2
    #3
    #4
    #5
    #6
   ]}
  n15 = GetMaterialAttributes(None=@n14) {attributeGetTypes [#1]}
  n16 = CollectionParameter() {collection=&MPC_ImposterCapture, parameterName=CustomLighting}
  n17 = GetMaterialAttributes(None=SetMaterialAttributes!(MaterialAttributes=@n15.MaterialAttributes, BaseColor=Add(A=@n15.BaseColor, B=@n16)) {attributeSetTypes [#1]}) {attributeGetTypes [#4]}
  n18 = FunctionInput(Preview=TwoSidedSign()) {bUsePreviewValueAsDefault=true, inputName=Two Sided Sign, inputType=FunctionInput_Scalar, sortPriority=8}
  n19 = Normalize(VectorInput=GetMaterialAttributes(None=@n1) {attributeGetTypes [#7]}.Normal)
  n20 = StaticSwitch(
   True=Transform(None=Multiply(A=@n18, B=@n19))
   False=@n19
   Value=FunctionInput(Preview=StaticBool() {value=true}) {bUsePreviewValueAsDefault=true, inputName=IsTangentSpaceNormal, inputType=FunctionInput_StaticBool, sortPriority=4}
  )
  n21 = Add(
   A=@n11
   B=Multiply(A=CollectionParameter() {collection=&MPC_ImposterCapture, parameterName=Normal}, B=ConstantBiasScale(None=@n20))
  )
  n22 = CollectionParameter() {collection=&MPC_ImposterCapture, parameterName=Origin}
  n23 = MaterialFunctionCall!(
   VectorToTransform=Subtract(A=WorldPosition().XYZ, B=ComponentMask(None=@n22) {b=true, g=true, r=true})
   BasisX=ComponentMask(None=CollectionParameter() {collection=&MPC_ImposterCapture, parameterName=X}) {b=true, g=true, r=true}
   BasisY=ComponentMask(None=CollectionParameter() {collection=&MPC_ImposterCapture, parameterName=Y}) {b=true, g=true, r=true}
   BasisZ=ComponentMask(None=CollectionParameter() {collection=&MPC_ImposterCapture, parameterName=Z}) {b=true, g=true, r=true}
  ) {materialFunction=/Engine/Functions/Engine_MaterialFunctions02/Math/InverseTransformMatrix}
  n24 = CollectionParameter() {collection=&MPC_ImposterCapture, parameterName=R}
  n25 = ComponentMask(None=ConstantBiasScale(None=Divide(A=@n23.Transformed Vector 3, B=@n24))) {b=true}
  n26 = Multiply(
   A=OneMinus(None=Saturate(None=@n25))
   B=CollectionParameter() {collection=&MPC_ImposterCapture, parameterName=Depth}
  )
  n27 = Add(
   A=@n17.EmissiveColor
   B=Multiply(A=Add(A=@n21, B=@n26), B=CollectionParameter() {collection=&MPC_ImposterCapture, parameterName=EnableCapture})
  )
  FunctionOutput(None=StaticSwitchParameter(
    True=SetMaterialAttributes!(MaterialAttributes=@n17.MaterialAttributes, EmissiveColor=@n27) {attributeSetTypes [#4]}
    False=@n1
   ) {defaultValue=true, parameterName=Enable Capture Function})
  n28 = CollectionParameter() {collection=&MPC_ImposterCapture, parameterName=X}
  n29 = Constant() {r=7}
  n30 = ScreenPosition()
  n31 = VectorParameter() {parameterName=Debug1, defaultValue=<rgba 0.2 0.2 0.4 0.5>}
  n32 = CollectionParameter() {collection=&MPC_ImposterCapture, parameterName=Y}
  n33 = VectorParameter() {parameterName=Debug2, defaultValue=<rgba 0.4 0.2 0.6 0.5>}
  n34 = Add(
   A=MaterialFunctionCall!(
    Vector3=ComponentMask(None=@n28) {b=true, g=true, r=true}
    MaximumNumberOfDigits=@n29
    UVs=@n30.ViewportUV
    DebugTextLocation RG_UpperRight BA_LowerLeft=AppendVector(A=@n31.RGB, B=@n31.RGB)
    Component Spacing
   ) {materialFunction=/Engine/Functions/Engine_MaterialFunctions02/Utility/DebugFloat3Values}.ColorCodedOutput
   B=MaterialFunctionCall(
    Vector3=ComponentMask(None=@n32) {b=true, g=true, r=true}
    MaximumNumberOfDigits=@n29
    UVs=@n30.ViewportUV
    DebugTextLocation RG_UpperRight BA_LowerLeft=AppendVector(A=@n33.RGB, B=@n33.RGB)
   ) {materialFunction=&DebugFloat3Values}.ColorCodedOutput
  )
  n35 = VectorParameter() {parameterName=Debug3, defaultValue=<rgba 0.3 0.5 0.5 0.8>}
  n36 = MaterialFunctionCall(
   Vector3=ComponentMask(None=CollectionParameter() {collection=&MPC_ImposterCapture, parameterName=Z}) {b=true, g=true, r=true}
   MaximumNumberOfDigits=@n29
   UVs=@n30.ViewportUV
   DebugTextLocation RG_UpperRight BA_LowerLeft=AppendVector(A=@n35.RGB, B=@n35.RGB)
  ) {materialFunction=&DebugFloat3Values}
  NamedRerouteDeclaration(None=Desaturation(None=Add(A=@n34, B=@n36.ColorCodedOutput), Fraction=Constant() {r=0.5})) {name=Debug Vectors, nodeColor=<rgba 0.1016 1 0 1>}
 ]
}
