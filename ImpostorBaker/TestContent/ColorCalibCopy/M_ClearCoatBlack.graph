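# M_ClearCoatBlack is a black, zero-roughness Clear Coat test material in the ImpostorBaker plugin's color calibration test content.
# Constant inputs for base color, roughness and full clear coat are packed through MakeMaterialAttributes.
# Those attributes feed the ImposterCaptureSwitch material function, the hook that by its name adapts the output for impostor capture.
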Material /ImpostorBaker/TestContent/ColorCalibCopy/M_ClearCoatBlack
graph {
 properties {
  bUseMaterialAttributes=true
  bUsedWithStaticLighting=true
  pixelDepthOffsetMode=PDOM_Legacy
  shadingModel=MSM_ClearCoat
  usedShadingModels=Clear Coat
 }
 nodes [MaterialFunctionCall!(
   Material Attributes=MakeMaterialAttributes!(
    BaseColor=Constant()
    Metallic
    Specular
    Roughness=Constant()
    Anisotropy
    EmissiveColor
    Opacity
    OpacityMask
    Normal
    Tangent
    WorldPositionOffset
    SubsurfaceColor
    ClearCoat=Constant() {r=1}
    ClearCoatRoughness
    AmbientOcclusion
    Refraction
    CustomizedUVs_0
    CustomizedUVs_1
    CustomizedUVs_2
    CustomizedUVs_3
    CustomizedUVs_4
    CustomizedUVs_5
    CustomizedUVs_6
    CustomizedUVs_7
    PixelDepthOffset
    ShadingModel
    Displacement
   )
   IsTangentSpaceNormal
   Two Sided Sign
  ) {materialFunction=/ImpostorBaker/MaterialFunctions/ImposterCaptureSwitch}]
}
